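# Post-process material that composites a Source texture into selected channels of a Target texture, apparently to compile biome mask textures by quadrant for environment art tools.
# Target UVs take scale and shift parameters; a similar Source transform exists but does not feed its sample.
# TargetChannelMask keeps Target channels where set; elsewhere RGB takes the summed Source components and alpha takes Source red.
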Material /Game/NeverCook/ArtTools/Tech_Art/EnvironmentBiomeMasking/Data/CompileTextureQuadrant
graph {
 properties {
  blendMode=BLEND_Masked
  blendableOutputAlpha=true
  materialDomain=MD_PostProcess
  pixelDepthOffsetMode=PDOM_Legacy
  shadingModel=MSM_Unlit
  usedShadingModels=Unlit
 }
 nodes [
  n1 = TextureCoordinate()
  n2 = ScalarParameter() {defaultValue=1, parameterName=TargetCoordScale}
  n3 = ScalarParameter() {parameterName=TargetCoordShift_X}
  n4 = ScalarParameter() {parameterName=TargetCoordShift_Y}
  n5 = TextureSampleParameter2D(UVs=Add(
    A=Multiply(A=@n1, B=@n2) {constB=0.5}
    B=MaterialFunctionCall(X=@n3, Y=@n4) {materialFunction=&MakeFloat2}.Result
   )) {parameterName=Target, texture=&T_BlacksRGB}
  n6 = VectorParameter() {parameterName=TargetChannelMask, defaultValue=<rgba 1 1 1 1>}
  n7 = TextureSampleParameter2D() {parameterName=Source, texture=&T_BlacksRGB}
  n8 = MaterialFunctionCall(f3=@n7.RGB) {materialFunction=&AddComponents}
  Add(
   A=Multiply(A=@n5.RGB, B=@n6.RGB) {constB=0}
   B=Multiply(
    A=OneMinus(None=@n6.RGB)
    B=MaterialFunctionCall(X=@n8.f3, Y=@n8.f3, Z=@n8.f3) {materialFunction=&MakeFloat3}.Result
   )
  )
  Add(
   A=Multiply(A=TextureCoordinate(), B=ScalarParameter() {defaultValue=1, parameterName=SourceCoordScale}) {constB=0.5}
   B=MaterialFunctionCall(X=ScalarParameter() {parameterName=SourceCoordShift_X}, Y=ScalarParameter() {parameterName=SourceCoordShift_Y}) {materialFunction=&MakeFloat2}.Result
  )
  Add(A=Multiply(A=@n5.A, B=@n6.A), B=Multiply(A=OneMinus(None=@n6.A), B=@n7.R) {constB=0})
  Constant() {r=1}
 ]
}
